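# Physics asset for the TiltedParrot pickaxe mesh, defining capsule bodies and joints for the handle, a hanging strap and four leaves.
# The strap and each leaf form two-bone chains anchored to the handle, which is constrained to b_AR_Root with free swing and twist.
# Leaf joints allow only a few degrees of movement for a stiff flutter, while strap joints swing much wider, tightening at the second segment.
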PhysicsAsset /BRCosmetics/Weapons/FORT_Melee/Pickaxe_TiltedParrot/Meshes/TiltedParrot_Axe_Physics
bodies {
 Melee_TwoHanded_Handle [{shapeType=Capsule, length=58, radius=3, shapeName=Melee_TwoHanded_Handle_capsule, center=<1 0 14>}]
 b_AR_Root [{
   shapeType=Capsule
   length=24.2715
   radius=6.2452
   shapeName=b_AR_Root_capsule
   center=<-5.7756 2.5991 6.9637>
   rotation=<rot 79.491 5.432 -1.0277>
  }]
 dyn_leaf_a_01 [{shapeType=Capsule, length=1, radius=1, shapeName=dyn_leaf_a_01_capsule, rotation=<rot 90 0 0>}]
 dyn_leaf_a_02 [{shapeType=Capsule, length=1, radius=1, shapeName=dyn_leaf_a_02_capsule, rotation=<rot 90 0 0>}]
 dyn_leaf_b_01 [{shapeType=Capsule, length=1, radius=1, shapeName=dyn_leaf_b_01_capsule, rotation=<rot 90 0 0>}]
 dyn_leaf_b_02 [{shapeType=Capsule, length=1, radius=1, shapeName=dyn_leaf_b_02_capsule, rotation=<rot 90 0 0>}]
 dyn_leaf_c_01 [{shapeType=Capsule, length=1, radius=1, shapeName=dyn_leaf_c_01_capsule, rotation=<rot 90 0 0>}]
 dyn_leaf_c_02 [{shapeType=Capsule, length=1, radius=1, shapeName=dyn_leaf_c_02_capsule, rotation=<rot 90 0 0>}]
 dyn_leaf_d_01 [{shapeType=Capsule, length=1, radius=1, shapeName=dyn_leaf_d_01_capsule, rotation=<rot 90 0 0>}]
 dyn_leaf_d_02 [{shapeType=Capsule, length=1, radius=1, shapeName=dyn_leaf_d_02_capsule, rotation=<rot 90 0 0>}]
 dyn_strap_01 [{shapeType=Capsule, length=4, radius=1.5, shapeName=dyn_strap_01_capsule, center=<3 0 0>, rotation=<rot 90 0 0>}]
 dyn_strap_02 [{shapeType=Capsule, length=5, radius=1, shapeName=dyn_strap_02_capsule, center=<3 0 0>, rotation=<rot 90 0 0>}]
}
constraints [
 {bone1Name=dyn_strap_01, bone2Name=Melee_TwoHanded_Handle, swing1LimitDegrees=90, swing2LimitDegrees=50, twistLimitDegrees=45}
 {bone1Name=dyn_strap_02, bone2Name=dyn_strap_01, swing1LimitDegrees=40, swing2LimitDegrees=20, twistLimitDegrees=25}
 {bone1Name=dyn_leaf_a_01, bone2Name=Melee_TwoHanded_Handle, swing1LimitDegrees=3, swing2LimitDegrees=3, twistLimitDegrees=2}
 {bone1Name=dyn_leaf_a_02, bone2Name=dyn_leaf_a_01, swing1LimitDegrees=3, swing2LimitDegrees=3, twistLimitDegrees=2}
 {bone1Name=dyn_leaf_b_01, bone2Name=Melee_TwoHanded_Handle, swing1LimitDegrees=3, swing2LimitDegrees=3, twistLimitDegrees=2}
 {bone1Name=dyn_leaf_b_02, bone2Name=dyn_leaf_b_01, swing1LimitDegrees=3, swing2LimitDegrees=3, twistLimitDegrees=2}
 {bone1Name=dyn_leaf_c_01, bone2Name=Melee_TwoHanded_Handle, swing1LimitDegrees=3, swing2LimitDegrees=3, twistLimitDegrees=2}
 {bone1Name=dyn_leaf_c_02, bone2Name=dyn_leaf_c_01, swing1LimitDegrees=3, swing2LimitDegrees=3, twistLimitDegrees=2}
 {bone1Name=dyn_leaf_d_01, bone2Name=Melee_TwoHanded_Handle, swing1LimitDegrees=3, swing2LimitDegrees=3, twistLimitDegrees=2}
 {bone1Name=dyn_leaf_d_02, bone2Name=dyn_leaf_d_01, swing1LimitDegrees=3, swing2LimitDegrees=3, twistLimitDegrees=2}
 {
  bone1Name=Melee_TwoHanded_Handle
  bone2Name=b_AR_Root
  swing1LimitDegrees=45
  swing1Motion=Free
  swing2LimitDegrees=45
  swing2Motion=Free
  twistLimitDegrees=45
  twistMotion=Free
 }
]
