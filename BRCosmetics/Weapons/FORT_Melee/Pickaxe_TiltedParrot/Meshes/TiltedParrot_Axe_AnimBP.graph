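# Animation Blueprint for the TiltedParrot pickaxe that adds physics-driven secondary motion to its straps.
# Two RigidBody branches simulate in base bone space off b_AR_Root with different motion limits, each followed by additive strap bone offsets.
# A bool blend on bIsInFrontEnd selects between the branches.
# Each update, a deprecated gravity-override state function computes the world gravity fed to both simulations.
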AnimBlueprint /BRCosmetics/Weapons/FORT_Melee/Pickaxe_TiltedParrot/Meshes/TiltedParrot_Axe_AnimBP
target_skeleton=&/TiltedParrot_Axe_Skeleton
graphs [
 AnimGraph {nodes [
   n1 = Get Gravity Override()
   n2 = RigidBody!(
    ComponentPose=Local To Component(LocalPose=Input Pose() {name=InPose_0})
    bAlphaBoolEnabled:bool
    OverrideWorldGravity:Vector=@n1
    ExternalForce
   ) -> (Pose) {
    alphaInputType=Bool
    bClampLinearTranslationLimitToRefPose=true
    bOverrideWorldGravity=true
    overridePhysicsAsset=&/TiltedParrot_Axe_Physics
    simulationSpace=BaseBoneSpace
    alphaBoolBlend {blendOption=Linear}
    baseBoneRef {boneName=b_AR_Root}
    componentAppliedLinearAccClamp=<8000 8000 8000>
    componentLinearAccScale=<0.25 0.25 0.25>
    componentLinearVelScale=<0.43 0.43 0.43>
    simSpaceSettings {
     maxAngularAcceleration=1000
     maxAngularVelocity=500
     maxLinearAcceleration=500
     maxLinearVelocity=700
     worldAlpha=0.65
     externalLinearDragV=<0.2 0.2>
    }
   }
   n3 = "Transform (Modify) Bone - Bone: base_strap"!(ComponentPose=@n2, bAlphaBoolEnabled:bool=True, Translation="0,0,0", Rotation="90,-90,60", Scale="1,1,1") -> (Pose) {
    alphaInputType=Bool
    rotationMode=BMM_Additive
    alphaBoolBlend {blendOption=Linear}
    boneToModify {boneName=base_strap}
    rotation=<rot 90 -90 60>
   }
   AnimGraphNode_SubInput_0 = Input Pose()
   n4 = RigidBody(ComponentPose=Local To Component(LocalPose=@AnimGraphNode_SubInput_0), OverrideWorldGravity=@n1) {
    alphaInputType=Bool
    bOverrideWorldGravity=true
    overridePhysicsAsset=&TiltedParrot_Axe_Physics
    simulationSpace=BaseBoneSpace
    alphaBoolBlend {blendOption=Linear}
    baseBoneRef {boneName=b_AR_Root}
    componentAppliedLinearAccClamp=<8000 8000 8000>
    componentLinearAccScale=<0.3 0.3 0.3>
    componentLinearVelScale=<0.43 0.43 0.43>
    simSpaceSettings {
     maxAngularAcceleration=700
     maxAngularVelocity=3
     maxLinearAcceleration=500
     maxLinearVelocity=350
     worldAlpha=0.38
     externalLinearDragV=<0.2 0.2 0.2>
    }
   }
   n5 = "Transform (Modify) Bone - Bone: dyn_strap_01"!(ComponentPose=@n4, bAlphaBoolEnabled:bool=True, Translation="0.311,0.211,-0.176", Rotation="-23.363,-7.960,2.283", Scale="1,1,1") -> (Pose) {
    alphaInputType=Bool
    rotationMode=BMM_Additive
    rotationSpace=BCS_BoneSpace
    translationMode=BMM_Additive
    translationSpace=BCS_BoneSpace
    alphaBoolBlend {blendOption=Linear}
    boneToModify {boneName=dyn_strap_01}
    rotation=<rot -23.363 -7.96 2.2831>
    translation=<0.311 0.211 -0.176>
   }
   n6 = "Transform (Modify) Bone - Bone: dyn_strap_02"!(ComponentPose=@n5, bAlphaBoolEnabled:bool=True, Translation="0,0,0", Rotation="12.776,9.367,14.517", Scale="1,1,1") -> (Pose) {
    alphaInputType=Bool
    rotationMode=BMM_Additive
    rotationSpace=BCS_BoneSpace
    alphaBoolBlend {blendOption=Linear}
    boneToModify {boneName=dyn_strap_02}
    rotation=<rot 12.776 9.367 14.5171>
   }
   Output Pose(Result=Blend Poses by bool(BlendPose_0=Component To Local(ComponentPose=@n3), BlendPose_1=Component To Local(ComponentPose=@n6), bActiveValue=Get bIsInFrontEnd()) {blendType=Linear}) {layerGroup=None}
  ]}
 EventGraph {nodes [
   Event BlueprintUpdateAnimation()
   > n1 = Gravity Override State Function DEPRECATED(
    CharacterPartAnimInstance=Self-Reference()
    GameTimeInSeconds=Get Game Time in Seconds_0() -> (Game Time in Seconds_0:double)
    GravityOverrideParams=Get Gravity Override Params()
   )
   > Set Gravity Override(Gravity Override=@n1)
   TryGetPawnOwner()
  ]}
]
variables [
 Game Time in Seconds_0 {type=number}
 Gravity Override {type=Vector}
 Gravity Override Params {
  type=GravityOverrideParamsStruct
  default {emoteOrMelee {gravityOverride=<0 0 -800>}, onGround_Moving {gravityOverride=<0 0 -800>}, onGround_Standing {gravityOverride=<0 0 -800>}}
 }
]
